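; Vehicle sound component for the SuperSedan (Valet car) that plays handbrake and catch-up boost audio.
; Binds car handbrake and seat delegates on init, spawning attached start, stop and catch-up sounds and cross-fading between them.
; A looping handbrake timer activates AudioController redlining and fades start audio at low speed.
; A control bus mix ducks or shapes audio for local passengers; driver exit or EndPlay resets audio.
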(event VehicleAudio|EventOnInit
  (bind _returnvalue (VehicleSound|GetVehicleActor))
  ;[1 1 1] Custom setup due to this being valet specific
  (|CallParentFunction)
  (Vehicle|AssignOnPawnExitedVehicleSeat _returnvalue (EventDispatchers|CreateEvent :!event Custom|PawnExitSeat))
  (Vehicle|AssignOnPawnEnteredVehicleSeat _returnvalue (EventDispatchers|CreateEvent :!event Custom|PawnEnterSeat))
  (bind _asvalet_basic_car_vehicle (Utilities|Casting|CastToValet_BasicCar_Vehicle _returnvalue))
  (EventDispatchers|AssignOnStartHandBrakeCatchupForceDelegate _asvalet_basic_car_vehicle (EventDispatchers|CreateEvent :!event Custom|OnStartHandBrakeCatchupForce))
  (EventDispatchers|AssignOnStartHandbrakeDelegate _asvalet_basic_car_vehicle (EventDispatchers|CreateEvent :!event Custom|OnStartHandbrake))
  (EventDispatchers|AssignOnStopHandbrakeDelegate _asvalet_basic_car_vehicle (EventDispatchers|CreateEvent :!event Custom|OnStopHandbrake)))

(event VehicleAudio|EventOnShutdown
  (bind _returnvalue (VehicleSound|GetVehicleActor))
  ;[1 1 1] Custom setup due to this being valet specific
  (|CallParentFunction)
  (Utilities|IsValid _returnvalue
    (:Is%20Valid
      (Vehicle|UnbindEventfromOnPawnExitedVehicleSeat _returnvalue (EventDispatchers|CreateEvent :!event Custom|PawnExitSeat))
      (Vehicle|UnbindEventfromOnPawnEnteredVehicleSeat _returnvalue (EventDispatchers|CreateEvent :!event Custom|PawnEnterSeat))
      (bind _asvalet_basic_car_vehicle (Utilities|Casting|CastToValet_BasicCar_Vehicle _returnvalue))
      (EventDispatchers|UnbindEventfromOnStartHandBrakeCatchupForceDelegate _asvalet_basic_car_vehicle (EventDispatchers|CreateEvent :!event Custom|OnStartHandBrakeCatchupForce))
      (EventDispatchers|UnbindEventfromOnStartHandbrakeDelegate _asvalet_basic_car_vehicle (EventDispatchers|CreateEvent :!event Custom|OnStartHandbrake))
      (EventDispatchers|UnbindEventfromOnStopHandbrakeDelegate _asvalet_basic_car_vehicle (EventDispatchers|CreateEvent :!event Custom|OnStopHandbrake)))))

(event Custom|OnStartHandBrakeCatchupForce
  ;[1 1 1] Custom setup due to this being valet specific
  (bind _catchupboostaudio (Variables|Default|GetCatchUpBoostAudio)
    (:then
      (Audio|Components|Audio|FadeOut _catchupboostaudio 0.25 0.0 "Linear" :!class /Script/Engine.AudioComponent))
    (:else
      _))
  (bind _handbrakecatchupsound (Variables|Data|GetHandBrakeCatchUpSound)
    (:then
      (bind _returnvalue (Audio|SpawnSoundAttached _handbrakecatchupsound (VehicleSound|GetRootComponent)))
      (Variables|Default|SetCatchUpBoostAudio _returnvalue))
    (:else)))

(event Custom|OnStartHandbrake
  ;[1 1 1] Custom setup due to this being valet specific
  (CallFunction|UpdateHandBrakeAudio true)
  (bind _returnvalue (Utilities|Time|SetTimerbyEvent (EventDispatchers|CreateEvent :!event Custom|UpdateRedlining) (Variables|Data|GetRedlineTime) true))
  (Variables|Private|SetRedlineHandle _returnvalue))

(event Custom|UpdateRedlining
  ;[1 1 1] Custom setup due to this being valet specific
  (AudioController|SetRedlineActive (VehicleSound|GetAudioController) true)
  (if (not (> (CallFunction|GetSpeed) 5.0))
    (bind _handbrakestartaudio (Variables|Default|GetHandBrakeStartAudio)
      (:then
        (Audio|Components|Audio|FadeOut _handbrakestartaudio 0.25 0.0 "Linear" :!class /Script/Engine.AudioComponent))
      (:else))))

(event Custom|OnStopHandbrake
  ;[1 1 1] Custom setup due to this being valet specific
  (CallFunction|UpdateHandBrakeAudio false)
  (AudioController|SetRedlineActive (VehicleSound|GetAudioController) false)
  (Utilities|Time|ClearandInvalidateTimerbyHandle (Variables|Private|GetRedlineHandle)))

(event EventEndPlay
  (params (EndPlayReason :type EEndPlayReason%20Enum :default "Destroyed"))
  (bind _booststartcbm (Variables|Data|GetBoostStartCBM))
  ;[1 1 1] Custom setup due to this being valet specific
  (CallFunction|ResetAudio)
  (if (Pawn|IsLocallyViewed (stale Custom|PawnExitSeat PlayerPawn))
    (if (Audio|Modulation|IsControlBusMixActive _booststartcbm)
      (Audio|Modulation|DeactivateControlBusMix _booststartcbm)
      (Variables|Default|SetPassengerIsLocal false))))

(event Custom|PawnExitSeat
  (params (Vehicle :type Fort%20Vehicle%20Interface%20Interface%20%28by%20ref%29) (PlayerPawn :type Fort%20Pawn%20Object%20Reference) (SeatIndex :type Integer))
  (bind _booststartcbm (Variables|Data|GetBoostStartCBM))
  ;[1 1 1] Custom setup due to this being valet specific
  (Default|UpdateAreAnyPassengersLocal Vehicle)
  (if (== SeatIndex 0)
    (CallFunction|ResetAudio))
  (if (Pawn|IsLocallyViewed PlayerPawn)
    (if (Audio|Modulation|IsControlBusMixActive _booststartcbm)
      (Audio|Modulation|DeactivateControlBusMix _booststartcbm)
      (Variables|Default|SetPassengerIsLocal false))))

(event Custom|PawnEnterSeat
  (params (Vehicle :type Fort%20Vehicle%20Interface%20Interface%20%28by%20ref%29) (PlayerPawn :type Fort%20Pawn%20Object%20Reference) (SeatIndex :type Integer))
  ;[1 1 1] Custom setup due to this being valet specific
  (Default|UpdateAreAnyPassengersLocal Vehicle))

(fn GetSpeed
  (return (Vehicle|GetTotalSpeedKmh (AudioController|GetVehicleActor (VehicleSound|GetAudioController)))))

(fn ResetAudio
  (AudioController|SetRedlineActive (VehicleSound|GetAudioController) false)
  (Utilities|Time|ClearandInvalidateTimerbyHandle (Variables|Private|GetRedlineHandle))
  (bind _handbrakestartaudio (Variables|Default|GetHandBrakeStartAudio)
    (:then
      (Audio|Components|Audio|FadeOut _handbrakestartaudio 0.15 0.0 "Linear" :!class /Script/Engine.AudioComponent))
    (:else
      _))
  (bind _catchupboostaudio (Variables|Default|GetCatchUpBoostAudio)
    (:then
      (Audio|Components|Audio|FadeOut _catchupboostaudio 0.15 0.0 "Linear" :!class /Script/Engine.AudioComponent))
    (:else
      _))
  (bind _handbrakestopaudio (Variables|Default|GetHandBrakeStopAudio)
    (:then
      (Audio|Components|Audio|FadeOut _handbrakestopaudio 0.15 0.0 "Linear" :!class /Script/Engine.AudioComponent))
    (:else)))

(fn UpdateAreAnyPassengersLocal
  (params (Target :type Fort%20Vehicle%20Interface%20Interface))
  (Variables|Default|SetPassengerIsLocal false)
  (for _array_element (Vehicle|GetAllPassengers Target)
    (if (Pawn|IsLocallyControlled _array_element)
      (Variables|Default|SetPassengerIsLocal true))))

(fn UpdateHandBrakeAudio
  (params (ShouldHandbrake :type Boolean))
  (if ShouldHandbrake
    ;[1 1 1] Start Handbrake audio section
    (if (> (CallFunction|GetSpeed) 5.0)
      (bind _handbrakestartaudio (Variables|Default|GetHandBrakeStartAudio)
        (:then
          (Audio|Components|Audio|FadeOut _handbrakestartaudio 0.25 0.0 "Linear" :!class /Script/Engine.AudioComponent))
        (:else
          _))
      (bind _handbrakestartsound (Variables|Data|GetHandBrakeStartSound)
        (:then
          (bind _returnvalue (Audio|SpawnSoundAttached _handbrakestartsound (VehicleSound|GetRootComponent)))
          (Variables|Default|SetHandBrakeStartAudio _returnvalue))
        (:else
          _))
      (bind _catchupboostaudio (Variables|Default|GetCatchUpBoostAudio)
        (:then
          (Audio|Components|Audio|FadeOut _catchupboostaudio 0.25 0.0 "Linear" :!class /Script/Engine.AudioComponent))
        (:else
          _))
      (bind _handbrakestopaudio (Variables|Default|GetHandBrakeStopAudio)
        (:then
          (Audio|Components|Audio|FadeOut _handbrakestopaudio 0.25 0.0 "Linear" :!class /Script/Engine.AudioComponent))
        (:else
          _))
      (if (Variables|Default|GetPassengerIsLocal)
        (Audio|Modulation|ActivateControlBusMix (Variables|Data|GetBoostStartCBM))))
    (else
      ;[1 1 1] Stop Handbrake Audio Section
      (bind _handbrakestartaudio_2 (Variables|Default|GetHandBrakeStartAudio)
        (:then
          (Audio|Components|Audio|FadeOut _handbrakestartaudio_2 0.25 0.0 "Linear" :!class /Script/Engine.AudioComponent))
        (:else
          _))
      (bind _handbrakestopsound (Variables|Data|GetHandBrakeStopSound)
        (:then
          (bind _returnvalue_2 (Audio|SpawnSoundAttached _handbrakestopsound (VehicleSound|GetRootComponent)))
          (Variables|Default|SetHandBrakeStopAudio _returnvalue_2))
        (:else
          _))
      (if (Variables|Default|GetPassengerIsLocal)
        (Audio|Modulation|DeactivateControlBusMix (Variables|Data|GetBoostStartCBM))))))
